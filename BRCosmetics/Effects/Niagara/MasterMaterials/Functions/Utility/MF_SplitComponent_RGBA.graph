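# Material function in the Niagara effects master-material utilities that splits a Vector4 input into its channels.
# A component mask takes RGB from the input, and further masks pull R, G and B from that result.
# Alpha is masked straight from the input, and the function outputs R, G, B, A, the combined RGB and the unchanged RGBA.
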MaterialFunction /BRCosmetics/Effects/Niagara/MasterMaterials/Functions/Utility/MF_SplitComponent_RGBA
graph {nodes [
  n1 = FunctionInput() {inputName=Vector 4, inputType=FunctionInput_Vector4, previewValue=<1 1 1 1>}
  n2 = ComponentMask(None=@n1) {b=true, g=true, r=true}
  FunctionOutput(None=ComponentMask(None=@n2) {r=true}) {outputName=R}
  FunctionOutput(None=ComponentMask(None=@n2) {g=true}) {outputName=G, sortPriority=1}
  FunctionOutput(None=ComponentMask(None=@n2) {b=true}) {outputName=B, sortPriority=2}
  FunctionOutput(None=ComponentMask(None=@n1) {a=true}) {outputName=A, sortPriority=3}
  FunctionOutput(None=@n2) {outputName=RGB, sortPriority=-1}
  FunctionOutput(None=@n1) {outputName=RGBA, sortPriority=5}
 ]}
