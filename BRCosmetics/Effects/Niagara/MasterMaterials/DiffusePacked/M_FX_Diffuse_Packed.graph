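# Masked, lit master material for Niagara mesh, sprite and ribbon effects, with static switches gating most features.
# Dynamic parameters or scalars pick packed diffuse and displacement patterns, tinted by primary and secondary colors.
# A shared packed texture drives UV distortion, cracks and erosion; cracks and curl-noise glitter alter color and surface response.
# Noise displaces vertices along normals, with a sine-rotation alternative, and normals can be rebuilt as faceted.
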Material /BRCosmetics/Effects/Niagara/MasterMaterials/DiffusePacked/M_FX_Diffuse_Packed
graph {
 properties {
  bAutomaticallySetUsageInEditor=false
  bHasPixelAnimation=true
  bUsedWithNiagaraMeshParticles=true
  bUsedWithNiagaraRibbons=true
  bUsedWithNiagaraSprites=true
  blendMode=BLEND_Masked
  usedShadingModels=Default Lit
 }
 nodes [
  MaterialFunctionCall!(Vector 4=NamedRerouteUsage()) {materialFunction=/BRCosmetics/Effects/Niagara/MasterMaterials/Functions/Utility/MF_SplitComponent_RGBA}
  n1 = TextureSampleParameter2D(UVs=NamedRerouteUsage()) {
   desc="Option to have diffuse Packed or not\r\nUse MaskPicker to select channel if packed"
   group="1 - Color && Pattern"
   parameterName=Diffuse
   sortPriority=31
   texture=/BRCosmetics/Effects/Niagara/MasterMaterials/DiffusePacked/Textures/T_FX_Packed_Patterns_01
  }
  n2 = StaticSwitchParameter(
   True=DynamicParameter() {defaultValue=<rgba 3 1 1 1>, paramNames [Mask Picker, WPO Picker, WPO Strength, Erosion]}.Mask Picker
   False=ScalarParameter() {defaultValue=3, group="1 - Color && Pattern", parameterName=Channel Picker, sliderMax=3, sortPriority=30}
  ) {
   defaultValue=true
   desc=If True reads from Dyanmic Paramter to pick channel for packed pattern diffuse. Otherwise it can be explicitly set with the Channel Picker parameter
   group="1 - Color && Pattern"
   parameterName=Use Niagara Pick Packed
   sortPriority=29
  }
  n3 = MaterialFunctionCall!(Mask 1=@n1.R, Mask 2=@n1.R, Mask 3=@n1.R, Mask 4=@n1.R, Blend=@n2) {materialFunction=/Game/Effects/Niagara/MasterMaterials/Functions/Utility/MF_Lerp_4_Scalar}
  n4 = TextureSampleParameter2D(UVs=NamedRerouteUsage()) {
   desc="Option to have diffuse Packed or not\r\nUse MaskPicker to select channel if packed"
   group="1 - Color && Pattern"
   parameterName=Diffuse
   sortPriority=31
   texture=&T_FX_Packed_Patterns_01
  }
  n5 = StaticSwitchParameter(
   True=MaterialFunctionCall(R=@n3.Result, G=@n3.Result, B=@n3.Result) {materialFunction=&AppendMany}.RGB
   False=@n4.RGB
  ) {defaultValue=true, group="1 - Color && Pattern", parameterName=Is Diffuse Packed, sortPriority=28}
  n6 = VectorParameter() {group="1 - Color && Pattern", parameterName=Primary Color, sortPriority=33, defaultValue=<rgba 0.051 0.8649 1 1>}
  n7 = StaticSwitchParameter(True=ParticleColor().RGBA, False=AppendVector(A=@n6.RGB, B=@n6.RGB)) {group="1 - Color && Pattern", parameterName=Use Particle Color}
  n8 = Multiply(A=@n5, B=ComponentMask(None=@n7) {b=true, g=true, r=true})
  n9 = NamedRerouteUsage()
  MaterialFunctionCall(Emissive=Multiply(
    A=StaticSwitchParameter(True=@n8, False=@n9) {group="1 - Color && Pattern", parameterName=Use Diffuse As Emissive Mask, sortPriority=37}
    B=ScalarParameter() {group="1 - Color && Pattern", parameterName=Emissive Strength, sortPriority=38}
   )) {materialFunction=&MF_FX_EmissiveAdjustments}
  n10 = Multiply(
   A=ScalarParameter() {defaultValue=0.1, group=4 - UV Distortion, parameterName=UV Distortion Strength Diffuse, sortPriority=33}
   B=NamedRerouteUsage()
  )
  n11 = MaterialFunctionCall!(
   UV=TextureCoordinate()
   Tile U=ScalarParameter() {defaultValue=2, group=1.5 - Surface UVs, parameterName=Tile U}
   Tile V=ScalarParameter() {defaultValue=1, group=1.5 - Surface UVs, parameterName=Tile V, sortPriority=33}
   Pan U=ScalarParameter() {group=1.5 - Surface UVs, parameterName=Pan Speed U, sortPriority=34}
   Pan V=ScalarParameter() {group=1.5 - Surface UVs, parameterName=Pan Speed V, sortPriority=35}
   Offset U=ScalarParameter() {group=1.5 - Surface UVs, parameterName=Offset U, sortPriority=36}
   Offset V=ScalarParameter() {group=1.5 - Surface UVs, parameterName=Offset V, sortPriority=37}
   Time=StaticSwitchParameter(True=ParticleRelativeTime(), False=Time()) {group=0 - Shared, parameterName=Use Particle Time}
  ) {materialFunction=/Game/Effects/Fort_Effects/Materials/_Functions/MF_PanTileOffset}
  NamedRerouteDeclaration(None=StaticSwitchParameter(True=Add(A=@n10, B=@n11.UV's), False=@n11.UV's) {defaultValue=true, group=4 - UV Distortion, parameterName=Use UV Distortion}) {name=Surface UVs, nodeColor=<rgba 0.1953 1 0 1>}
  n12 = ParticleRelativeTime()
  n13 = Time()
  NamedRerouteDeclaration(None=MaterialFunctionCall(
    UV=TextureCoordinate()
    Tile U=ScalarParameter() {defaultValue=1, group=7.5 - WPO UVs, parameterName=WPO -Tile U}
    Tile V=ScalarParameter() {defaultValue=1, group=7.5 - WPO UVs, parameterName=WPO -Tile V, sortPriority=33}
    Pan U=ScalarParameter() {defaultValue=0.01, group=7.5 - WPO UVs, parameterName=WPO - Pan Speed U, sortPriority=34}
    Pan V=ScalarParameter() {group=7.5 - WPO UVs, parameterName=WPO - Pan Speed V, sortPriority=35}
    Offset U=ScalarParameter() {group=7.5 - WPO UVs, parameterName=WPO - Offset U, sortPriority=36}
    Offset V=ScalarParameter() {group=7.5 - WPO UVs, parameterName=WPO - Offset V, sortPriority=37}
    Time=StaticSwitchParameter(True=@n12, False=@n13) {group=0 - Shared, parameterName=Use Particle Time}
   ) {materialFunction=&MF_PanTileOffset}.UV's) {name=WPO UVs, nodeColor=<rgba 1 0.8438 0 1>}
  NamedRerouteDeclaration(None=@n3.Result) {name=Packed Dissue Mask, nodeColor=<rgba 0.4062 1 0 1>}
  n14 = Multiply(
   A=ScalarParameter() {defaultValue=0.1, group=4 - UV Distortion, parameterName=UV Distortion Strength Normal Map, sortPriority=34}
   B=NamedRerouteUsage()
  )
  n15 = MaterialFunctionCall(
   UV=TextureCoordinate()
   Tile U=ScalarParameter() {defaultValue=2, group=2.5 - Normal UVs, parameterName=Normal Tile U}
   Tile V=ScalarParameter() {defaultValue=1, group=2.5 - Normal UVs, parameterName=Normal Tile V, sortPriority=33}
   Pan U=ScalarParameter() {group=2.5 - Normal UVs, parameterName=Normal Pan Speed U, sortPriority=34}
   Pan V=ScalarParameter() {group=2.5 - Normal UVs, parameterName=Normal Pan Speed V, sortPriority=35}
   Offset U=ScalarParameter() {group=2.5 - Normal UVs, parameterName=Normal Offset U, sortPriority=36}
   Offset V=ScalarParameter() {group=2.5 - Normal UVs, parameterName=Normal Offset V, sortPriority=37}
   Time=StaticSwitchParameter(True=ParticleRelativeTime(), False=Time()) {group=0 - Shared, parameterName=Use Particle Time}
  ) {materialFunction=&MF_PanTileOffset}
  NamedRerouteDeclaration(None=StaticSwitchParameter(True=Add(A=@n14, B=@n15.UV's), False=@n15.UV's) {defaultValue=true, group=4 - UV Distortion, parameterName=Use UV Distortion}) {name=Normal UVs, nodeColor=<rgba 0.1953 1 0 1>}
  n16 = TextureSampleParameter2D(UVs=StaticSwitchParameter(True=NamedRerouteUsage(), False=NamedRerouteUsage()) {defaultValue=true, group=5.5 - Cracks UVs, parameterName=Use Unique Crack UVs, sortPriority=31}) {
   desc="RG - Uv Distortion \r\nB - Crack mask \r\nA - Erosion Pattern"
   group=0 - Shared
   parameterName="UV Distort, Crack && Erosion"
   texture=/BRCosmetics/Effects/Niagara/MasterMaterials/DiffusePacked/Textures/T_FX_Packed_UVCE_EggBounce
   channelNames {a=Erosion, b=Crack, g=UV Distort V, r=UV Distort U}
  }
  n17 = ScalarParameter() {defaultValue=100, group=5 - Cracks, parameterName=Crack Normal Strength, sortPriority=34}
  n18 = DynamicParameter() {parameterIndex=2, paramNames [Crack Alpha, Param2, Param3, Param4]}
  NamedRerouteDeclaration(None=MaterialFunctionCall!(Height=@n16.Crack, Strength=LinearInterpolate(B=@n17, Alpha=@n18.Crack Alpha)) {materialFunction=/BRCosmetics/Effects/Niagara/MasterMaterials/Functions/MF_FX_SimpleNormalFromMask}.RG) {name=Crack Normal, nodeColor=<rgba 0 1 0.6953 1>}
  n19 = Multiply(A=ScalarParameter() {group=4 - UV Distortion, parameterName=UV Distortion Strength Crack, sortPriority=35}, B=NamedRerouteUsage())
  n20 = MaterialFunctionCall(
   UV=TextureCoordinate()
   Tile U=ScalarParameter() {defaultValue=2, group=5.5 - Cracks UVs, parameterName=Crack Tile U}
   Tile V=ScalarParameter() {defaultValue=1, group=5.5 - Cracks UVs, parameterName=Crack Tile V, sortPriority=33}
   Pan U=ScalarParameter() {group=5.5 - Cracks UVs, parameterName=Crack Pan Speed U, sortPriority=34}
   Pan V=ScalarParameter() {group=5.5 - Cracks UVs, parameterName=Crack Pan Speed V, sortPriority=35}
   Offset U=ScalarParameter() {group=5.5 - Cracks UVs, parameterName=Crack Offset U, sortPriority=36}
   Offset V=ScalarParameter() {group=5.5 - Cracks UVs, parameterName=Crack Offset V, sortPriority=37}
   Time=StaticSwitchParameter(True=ParticleRelativeTime(), False=Time()) {group=0 - Shared, parameterName=Use Particle Time}
  ) {materialFunction=&MF_PanTileOffset}
  NamedRerouteDeclaration(None=StaticSwitchParameter(True=Add(A=@n19, B=@n20.UV's), False=@n20.UV's) {defaultValue=true, group=4 - UV Distortion, parameterName=Use UV Distortion}) {name=Crack UVs, nodeColor=<rgba 0 0.7422 1 1>}
  n21 = Multiply(
   A=ScalarParameter() {defaultValue=0.1, group=4 - UV Distortion, parameterName=UV Distortion Strength Erosion Map, sortPriority=36}
   B=NamedRerouteUsage()
  )
  n22 = MaterialFunctionCall(
   UV=TextureCoordinate()
   Tile U=ScalarParameter() {defaultValue=2, group=6.5 - Erosion UVs, parameterName=Erosion Tile U}
   Tile V=ScalarParameter() {defaultValue=1, group=6.5 - Erosion UVs, parameterName=Erosion Tile V, sortPriority=33}
   Pan U=ScalarParameter() {group=6.5 - Erosion UVs, parameterName=Erosion Pan Speed U, sortPriority=34}
   Pan V=ScalarParameter() {group=6.5 - Erosion UVs, parameterName=Erosion Pan Speed V, sortPriority=35}
   Offset U=ScalarParameter() {group=6.5 - Erosion UVs, parameterName=Erosion Offset U, sortPriority=36}
   Offset V=ScalarParameter() {group=6.5 - Erosion UVs, parameterName=Erosion Offset V, sortPriority=37}
   Time=StaticSwitchParameter(True=ParticleRelativeTime(), False=Time()) {group=0 - Shared, parameterName=Use Particle Time}
  ) {materialFunction=&MF_PanTileOffset}
  NamedRerouteDeclaration(None=StaticSwitchParameter(True=Add(A=@n21, B=@n22.UV's), False=@n22.UV's) {defaultValue=true, group=4 - UV Distortion, parameterName=Use UV Distortion}) {name=Erosion UVs, nodeColor=<rgba 1 0 0.5156 1>}
  n23 = Transform(None=Constant3Vector() {constant=<rgba -1 0 1 0>}) {transformSourceType=TRANSFORMSOURCE_Local}
  n24 = ComponentMask(None=TextureCoordinate()) {g=true}
  n25 = ScalarParameter() {defaultValue=1, group=9 - WPO_1, parameterName=WPO_1_Tile}
  n26 = Multiply(A=DynamicParameter() {parameterIndex=3, paramNames [WPO_1_Tile,Param2,Param3,Param4]}.WPO_1_Tile, B=@n25)
  n27 = Sine(None=Add(A=Multiply(A=@n24, B=@n26), B=Time()))
  n28 = ConstantBiasScale(None=@n27)
  n29 = Transform(None=Constant3Vector() {constant=<rgba 0 1 1 0>}) {transformSourceType=TRANSFORMSOURCE_Local}
  n30 = ComponentMask(None=TextureCoordinate()) {g=true}
  n31 = AppendVector(A=Normalize(VectorInput=LinearInterpolate(A=@n23, B=@n29, Alpha=@n28)), B=Multiply(A=Multiply(A=@n25, B=@n27), B=OneMinus(None=@n30)))
  n32 = MaterialFunctionCall() {materialFunction=&ObjectPivotPoint}
  n33 = WorldPosition()
  NamedRerouteDeclaration(None=Multiply(
    A=RotateAboutAxis(NormalizedRotationAxis=@n31, RotationAngle=@n31, PivotPoint=@n32.Mesh Particle Pivot Location, Position=@n33.XYZ)
    B=ScalarParameter() {group=9 - WPO_1, parameterName=WPO_1_Inentsity}
   )) {name=WPO_1, nodeColor=<rgba 0 1 0.1328 1>}
  n34 = WorldPosition()
  n35 = DDX(Value=@n34.XYZ)
  n36 = DDY(Value=@n34.XYZ)
  n37 = ObjectPositionWS()
  n38 = Normalize(VectorInput=StaticSwitchParameter(True=CrossProduct(A=@n35, B=@n36), False=Subtract(A=@n34.XYZ, B=@n37)) {group=7 - WPO, parameterName=Use Faceted Normals, sortPriority=37})
  n39 = TextureSampleParameter2D(UVs=StaticSwitchParameter(True=NamedRerouteUsage(), False=NamedRerouteUsage()) {group=2.5 - Normal UVs, parameterName=Use Unique Normal UVs, sortPriority=31}) {
   desc=Standard Normal Map
   group=2 - Normal
   parameterName=Normal
   samplerType=SAMPLERTYPE_Normal
   texture=/Game/Creative/Devices/Eggs/Textures/Foil_P_1/Foil_Egg_N
  }
  n40 = StaticSwitchParameter(
   True=Multiply(
    A=ComponentMask(None=@n39.RGB) {g=true, r=true}
    B=ScalarParameter() {defaultValue=1, group=2 - Normal, parameterName=Normal Strength, sortPriority=34}
   )
   False=ComponentMask(None=Constant3Vector() {constant=<rgba 0.5 0.5 1 0>}) {g=true, r=true}
  ) {defaultValue=true, group=2 - Normal, parameterName=Use Normal Map, sortPriority=31}
  n41 = StaticSwitchParameter(True=Add(A=NamedRerouteUsage(), B=@n40), False=@n40) {defaultValue=true, group=5 - Cracks, parameterName=Use Crack, sortPriority=31}
  n42 = Add(
   A=ComponentMask(None=Transform(None=@n38) {transformSourceType=TRANSFORMSOURCE_World, transformType=TRANSFORM_Tangent}) {g=true, r=true}
   B=@n41
  )
  NamedRerouteDeclaration(None=AppendVector(
    A=StaticSwitchParameter(True=@n42, False=@n41) {defaultValue=true, group=7 - WPO, parameterName=Use WPO, sortPriority=31}
    B=Constant() {r=1}
   )) {name=Normals, nodeColor=<rgba 0.7578 1 0 1>}
  n43 = TextureSampleParameter2D(UVs=NamedRerouteUsage()) {
   desc="Spec, Metal, Roughness, Anisotropy - option for texture or no texture SMRA setting"
   group="3 - Spec, Metal, Roughness, Anisotropy"
   parameterName=SMRA
   samplerType=SAMPLERTYPE_Masks
   texture=/Game/Creative/Devices/Eggs/Textures/Foil_P_4/Foil_Egg_S
   channelNames {a=Anisotropy, b=Roughness, g=Metallic, r=Specular}
  }
  n44 = ScalarParameter() {group="3 - Spec, Metal, Roughness, Anisotropy", parameterName=Anisotropy, sortPriority=37}
  n45 = StaticSwitchParameter(True=AppendVector(A=@n43.RGB, B=@n44), False=@n43.RGBA) {defaultValue=true, group="3 - Spec, Metal, Roughness, Anisotropy", parameterName=Override Textuure Anisotropy, sortPriority=33}
  n46 = StaticSwitchParameter(
   True=@n45
   False=MaterialFunctionCall(
    R=ScalarParameter() {defaultValue=0.5, group="3 - Spec, Metal, Roughness, Anisotropy", parameterName=Specular, sortPriority=34}
    G=ScalarParameter() {defaultValue=1, group="3 - Spec, Metal, Roughness, Anisotropy", parameterName=Metallic, sortPriority=35}
    B=ScalarParameter() {group="3 - Spec, Metal, Roughness, Anisotropy", parameterName=Roughness, sortPriority=36}
    A=@n44
   ) {materialFunction=&AppendMany}.RGBA
  ) {group="3 - Spec, Metal, Roughness, Anisotropy", parameterName=Use SMRA Texture, sortPriority=31}
  n47 = NamedRerouteUsage()
  n48 = NamedRerouteUsage()
  n49 = OneMinus(None=@n48)
  n50 = StaticSwitchParameter(
   True=LinearInterpolate(
    A=ComponentMask(None=@n46) {g=true, r=true}
    B=@n47
    Alpha=StaticSwitchParameter(True=@n49, False=@n48) {group=8 - Glitter, parameterName=Invert Diffuse Mask for Glitter Mask, sortPriority=36}
   )
   False=AppendVector(A=@n47, B=@n47)
  ) {defaultValue=true, group=8 - Glitter, parameterName=Use Diffuse As Glitter Mask, sortPriority=35}
  n51 = Subtract(A=ComponentMask(None=@n46) {b=true}, B=@n47)
  n52 = StaticSwitchParameter(True=AppendVector(A=AppendVector(A=@n50, B=@n51), B=ComponentMask(None=@n46) {a=true}), False=@n46) {defaultValue=true, group=8 - Glitter, parameterName=UseGlitter, sortPriority=31}
  n53 = NamedRerouteUsage()
  n54 = MaterialFunctionCall(
   R=Multiply(A=@n53, B=ScalarParameter() {defaultValue=0.5, group=5 - Cracks, parameterName=Crack Spec, sortPriority=36})
   G=Multiply(A=@n53, B=ScalarParameter() {group=5 - Cracks, parameterName=Crack Metal, sortPriority=37})
   B=Multiply(A=@n53, B=ScalarParameter() {defaultValue=0.5, group=5 - Cracks, parameterName=Crack Roughness, sortPriority=38})
   A=Multiply(A=@n53, B=ScalarParameter() {group=5 - Cracks, parameterName=Crack Anisotropy, sortPriority=39})
  ) {materialFunction=&AppendMany}
  NamedRerouteDeclaration(None=StaticSwitchParameter(True=LinearInterpolate(A=@n52, B=@n54.RGBA, Alpha=@n53), False=@n52) {defaultValue=true, group=5 - Cracks, parameterName=Use Crack, sortPriority=31}) {name=SMRA, nodeColor=<rgba 0 0.2969 1 1>}
  n55 = VectorParameter() {group=8 - Glitter, parameterName=Glitter Color, sortPriority=34, defaultValue=<rgba 1 0.4794 0.783 1>}
  n56 = NamedRerouteUsage()
  n57 = VectorParameter() {group="1 - Color && Pattern", parameterName=Secondary Color, sortPriority=36, defaultValue=<rgba 1 0.808 0.066 1>}
  n58 = StaticSwitchParameter(True=@n57.RGB, False=DynamicParameter() {parameterIndex=1, paramNames [Secondary R, Secondary G, Secondary B, Secondary A]}.RGB) {defaultValue=true, group="1 - Color && Pattern", parameterName=Use Material Secondary Color, sortPriority=35}
  n59 = OneMinus(None=@n5)
  n60 = StaticSwitchParameter(True=Add(A=Multiply(A=@n58, B=@n59), B=@n8), False=@n8) {defaultValue=true, group="1 - Color && Pattern", parameterName=Use Invert Diffuse As Secondary Color, sortPriority=34}
  n61 = LinearInterpolate(A=@n60, B=Multiply(A=@n55.RGB, B=@n56), Alpha=@n56)
  n62 = ScalarParameter() {group=8 - Glitter, parameterName=Glitter Color Blend, sortPriority=33}
  n63 = StaticSwitchParameter(
   True=StaticSwitchParameter(True=LinearInterpolate(A=@n60, B=@n61, Alpha=@n62), False=@n60) {defaultValue=true, group=8 - Glitter, parameterName=Use Glitter Sparkle Color}
   False=@n60
  ) {defaultValue=true, group=8 - Glitter, parameterName=UseGlitter, sortPriority=31}
  n64 = VectorParameter() {group=5 - Cracks, parameterName=Crack Color, sortPriority=35, defaultValue=<rgba 0.0885 0.0665 0.0517 1>}
  n65 = NamedRerouteUsage()
  n66 = Multiply(A=@n64.RGB, B=@n65)
  NamedRerouteDeclaration(None=StaticSwitchParameter(True=LinearInterpolate(A=@n63, B=@n66, Alpha=@n65) {constAlpha=1}, False=@n63) {defaultValue=true, group=5 - Cracks, parameterName=Use Crack, sortPriority=31}) {name=Color, nodeColor=<rgba 0 0.4375 1 1>}
  n67 = TextureSampleParameter2D(UVs=NamedRerouteUsage()) {
   desc=Packed WPO - each channel is a differnt pattern that can be chosen
   group=7 - WPO
   parameterName=WPO
   samplerType=SAMPLERTYPE_LinearColor
   texture=/BRCosmetics/Effects/Niagara/MasterMaterials/DiffusePacked/Textures/T_FX_Packed_WPO_Noise_01
  }
  n68 = MaterialFunctionCall(
   Mask 1=@n67.R
   Mask 2=@n67.R
   Mask 3=@n67.R
   Mask 4=@n67.R
   Blend=StaticSwitchParameter(
    True=DynamicParameter() {defaultValue=<rgba 3 1 1 1>, paramNames [Mask Picker, WPO Picker, WPO Strength, Erosion]}.WPO Picker
    False=ScalarParameter() {defaultValue=2, group=7 - WPO, parameterName=WPO Picker, sliderMax=3, sortPriority=34}
   ) {group=7 - WPO, parameterName=Use Niagara WPO Pick Packed, sortPriority=33}
  ) {materialFunction=&MF_Lerp_4_Scalar}
  n69 = StaticSwitchParameter(
   True=ScalarParameter() {defaultValue=100, group=7 - WPO, parameterName=WPO Strength, sortPriority=36}
   False=DynamicParameter() {defaultValue=<rgba 3 1 1 1>, paramNames [Mask Picker, WPO Picker, WPO Strength, Erosion]}.WPO Strength
  ) {group=7 - WPO, parameterName=Use Material WPO Strength, sortPriority=35}
  n70 = StaticSwitchParameter(True=Multiply(A=VertexNormalWS(), B=Multiply(A=@n68.Result, B=@n69) {constB=50}) {constB=0}, False=Constant3Vector()) {defaultValue=true, group=7 - WPO, parameterName=Use WPO, sortPriority=31}
  NamedRerouteDeclaration(None=If(A=ScalarParameter() {group=9 - WPO_1, parameterName=WPO/WPO_1}, "A > B"=NamedRerouteUsage(), "A == B"=@n70, "A < B"=@n70)) {name=WPO, nodeColor=<rgba 0 1 0.1328 1>}
  n71 = ParticleRelativeTime()
  n72 = Time()
  n73 = TextureSampleParameter2D(UVs=MaterialFunctionCall(
    UV=TextureCoordinate()
    Tile U=ScalarParameter() {defaultValue=1, group=4 - UV Distortion, parameterName=UV Distortion Tile U, sortPriority=37}
    Tile V=ScalarParameter() {defaultValue=1, group=4 - UV Distortion, parameterName=UV Distortion Tile V, sortPriority=38}
    Pan U=ScalarParameter() {group=4 - UV Distortion, parameterName=UV Distortion Pan Speed U, sortPriority=39}
    Pan V=ScalarParameter() {group=4 - UV Distortion, parameterName=UV Distortion Pan Speed V, sortPriority=40}
    Offset U=ScalarParameter() {group=4 - UV Distortion, parameterName=UV Distortion Offset U, sortPriority=41}
    Offset V=ScalarParameter() {group=4 - UV Distortion, parameterName=UV Distortion Offset V, sortPriority=42}
    Time=StaticSwitchParameter(True=@n71, False=@n72) {group=0 - Shared, parameterName=Use Particle Time}
   ) {materialFunction=&MF_PanTileOffset}.UV's) {
   desc="RG - Uv Distortion \r\nB - Crack mask \r\nA - Erosion Pattern"
   group=0 - Shared
   parameterName="UV Distort, Crack && Erosion"
   texture=&T_FX_Packed_UVCE_EggBounce
   channelNames {a=Erosion, b=Crack, g=UV Distort V, r=UV Distort U}
  }
  NamedRerouteDeclaration(None=ComponentMask(None=@n73.RGB) {g=true, r=true}) {name=UV Distortion, nodeColor=<rgba 1 0.4922 0 1>}
  n74 = DynamicParameter() {defaultValue=<rgba 3 1 1 1>, paramNames [Mask Picker, WPO Picker, WPO Strength, Erosion]}
  n75 = ScalarParameter() {defaultValue=1, group=6 - Erosion, parameterName=Erosion Amount, sortPriority=34}
  n76 = ScalarParameter() {defaultValue=0.1, group=6 - Erosion, parameterName=Erosion Line Blur, sortPriority=35}
  n77 = Subtract(
   A=StaticSwitchParameter(
    True=ParticleColor().A
    False=StaticSwitchParameter(True=@n74.Erosion, False=@n75) {defaultValue=true, group=6 - Erosion, parameterName=Erosion - Use Dynamic Parameter, sortPriority=33}
   ) {defaultValue=true, group=6 - Erosion, parameterName=Erosion - Use Particle Alpha}
   B=@n76
  )
  n78 = Multiply(A=@n76) {constB=-1}
  n79 = TextureSampleParameter2D(UVs=StaticSwitchParameter(True=NamedRerouteUsage(), False=NamedRerouteUsage()) {group=6.5 - Erosion UVs, parameterName=Use Unique Erosion UVs, sortPriority=31}) {
   desc="RG - Uv Distortion \r\nB - Crack mask \r\nA - Erosion Pattern"
   group=0 - Shared
   parameterName="UV Distort, Crack && Erosion"
   texture=&T_FX_Packed_UVCE_EggBounce
   channelNames {a=Erosion, b=Crack, g=UV Distort V, r=UV Distort U}
  }
  NamedRerouteDeclaration(None=SmoothStep(Min=Add(A=@n77, B=@n76), Max=Add(A=@n77, B=@n78), Value=@n79.Erosion) {constMax=0.6, constMin=0.5}) {name=Erosion, nodeColor=<rgba 1 0 0.2812 1>}
  n80 = ScalarParameter() {defaultValue=1, group=5 - Cracks, parameterName=Crack Alpha, sortPriority=33}
  NamedRerouteDeclaration(None=Multiply(A=@n16.Crack, B=StaticSwitchParameter(True=@n80, False=@n18.Crack Alpha) {group=5 - Cracks, parameterName=Use Material Crack Alpha})) {name=Cracks Alpha, nodeColor=<rgba 0.9922 0 1 1>}
  n81 = NamedRerouteUsage()
  n82 = StaticSwitchParameter(True=OneMinus(None=@n81), False=@n81) {group=8 - Glitter, parameterName=Invert Diffuse Mask for Glitter Mask, sortPriority=36}
  n83 = Time() {bOverride_Period=true, period=1}
  n84 = VectorNoise(World Position=Add(A=WorldPosition().XYZ, B=Multiply(B=@n83) {constA=0.1}) {constB=0.1}) {bTiling=true, noiseFunction=VNF_CurlALU, tileSize=10}
  n85 = Multiply(
   A=StaticSwitchParameter(True=ParticleRelativeTime(), False=Time() {bOverride_Period=true, period=1}) {group=0 - Shared, parameterName=Use Particle Time}
   B=ScalarParameter() {defaultValue=3, group=8 - Glitter, parameterName=GlitterSparkleSpeed, sortPriority=38}
  )
  n86 = Constant() {r=3}
  n87 = MaterialFunctionCall(
   Mask 1=ComponentMask(None=@n84) {r=true}
   Mask 2=ComponentMask(None=@n84) {g=true}
   Mask 3=ComponentMask(None=@n84) {b=true}
   Mask 4=ComponentMask(None=@n84) {r=true}
   Blend=StaticSwitchParameter(True=Modulo!(A=@n85, B=@n86), False=ScalarParameter() {group=8 - Glitter, parameterName=GlitterSelection, sortPriority=37}) {group=8 - Glitter, parameterName=AnimateGlitterSparkle, sortPriority=37}
  ) {materialFunction=&MF_Lerp_4_Scalar}
  n88 = Add(A=Multiply(A=@n87.Result) {constB=0.5}, B=Round(None=@n87.Result))
  NamedRerouteDeclaration(None=StaticSwitchParameter(True=Multiply(A=@n82, B=@n88), False=@n88) {defaultValue=true, group=8 - Glitter, parameterName=Use Diffuse As Glitter Mask, sortPriority=35}) {name=Glitter, nodeColor=<rgba 1 0.9375 0 1>}
  NamedRerouteUsage()
  NamedRerouteUsage()
  NamedRerouteUsage()
  n89 = ComponentMask(None=@n7) {a=true}
  n90 = StaticSwitchParameter(True=@n3.Result, False=@n4.A) {defaultValue=true, group="1 - Color && Pattern", parameterName=Is Diffuse Packed, sortPriority=28}
  n91 = OneMinus(None=@n90)
  n92 = StaticSwitchParameter(
   True=Multiply(A=@n89, B=StaticSwitchParameter(True=@n91, False=@n90) {group="1 - Color && Pattern", parameterName=Flip Diffuse Alpha, sortPriority=40})
   False=@n89
  ) {group="1 - Color && Pattern", parameterName=Use Diffuse Alpha, sortPriority=39}
  StaticSwitchParameter(True=Multiply(A=NamedRerouteUsage(), B=@n92), False=@n92) {defaultValue=true, group=6 - Erosion, parameterName=Use Erosion Map, sortPriority=30}
 ]
}
